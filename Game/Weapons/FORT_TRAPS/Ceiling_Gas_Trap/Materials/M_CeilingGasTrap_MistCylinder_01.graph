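# Translucent, two-sided lit mesh-particle material for the ceiling gas trap's mist cylinder, with an IceTrap static switch that alters its emissive and vertex offset.
# A panning noise texture, a double-sided fresnel and a UV-gradient edge band shape color, opacity and a matching normal map.
# Dynamic parameters drive tiling, pan speed and normal-map-based vertex displacement, plus a flare along V.
# Opacity distance-fades; outside IceTrap, vertices collapse toward the pivot inside the end-fade distance.
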Material /Game/Weapons/FORT_TRAPS/Ceiling_Gas_Trap/Materials/M_CeilingGasTrap_MistCylinder_01
graph {
 properties {
  bEnableResponsiveAA=true
  bUsedWithMeshParticles=true
  bUsedWithNiagaraMeshParticles=true
  blendMode=BLEND_Translucent
  pixelDepthOffsetMode=PDOM_Legacy
  translucencyLightingMode=TLM_VolumetricPerVertexDirectional
  twoSided=true
  usedShadingModels=Default Lit
 }
 nodes [
  n1 = VectorParameter() {parameterName=Color A, defaultValue=<rgba 1.3 0.6847 0.0195 0>}
  n2 = VectorParameter() {parameterName=Color B, defaultValue=<rgba 1.5 1.1643 0.5625 0>}
  n3 = VertexColor()
  n4 = TextureCoordinate()
  n5 = Constant() {r=1}
  n6 = DynamicParameter() {paramNames [V Tile, Panner Speed, WPO Strength, Param4]}
  n7 = Panner(
   Coordinate=Multiply(A=Add(A=@n3.R, B=@n4), B=AppendVector(A=@n5, B=@n6.V Tile))
   Time=Multiply(A=@n6.Panner Speed, B=Time() {bOverride_Period=true, period=1})
  ) {bFractionalPart=true, speedX=0.335, speedY=-4}
  n8 = TextureSample(UVs=@n7) {texture=&2ndInfectedMapLinear}
  n9 = NamedRerouteUsage()
  n10 = MaterialFunctionCall(Float2=TextureCoordinate()) {materialFunction=&BreakOutFloat2Components}
  n11 = OneMinus(None=@n10.G)
  n12 = Power(Base=@n11) {constExponent=12}
  n13 = Saturate(None=Add(A=Multiply(A=@n8.RGB, B=@n9), B=@n12))
  n14 = Multiply(
   A=Multiply(A=LinearInterpolate(A=@n1.RGB, B=@n2.RGB, Alpha=@n13), B=ParticleColor().RGB)
   B=ScalarParameter() {defaultValue=0.08, parameterName=Emissive Boost}
  )
  n15 = NamedRerouteUsage()
  MaterialFunctionCall(Emissive=StaticSwitchParameter(True=@n14, False=Multiply(A=@n14, B=@n15)) {parameterName=IceTrap}) {materialFunction=&MF_FX_EmissiveAdjustments}
  n16 = Multiply(A=@n8.RGB, B=NamedRerouteUsage())
  n17 = ScalarParameter() {defaultValue=100, parameterName=End Fade}
  n18 = MaterialFunctionCall(
   Opacity In=Add(A=Multiply(A=@n16, B=@n11), B=@n12)
   Position=WorldPosition().XYZ
   Fade_Start_Distance=ScalarParameter() {defaultValue=200, parameterName=Start Fade}
   Fade_End_Distance=@n17
  ) {materialFunction=&MF_Distance_SyncedFade_FX}
  Multiply(A=@n18.Result, B=ParticleColor().A)
  n19 = TextureSample(UVs=@n7) {samplerType=SAMPLERTYPE_Normal, texture=/Game/Global/Textures/Default/Tileable_Patterns/2ndInfectedMapNormal}
  Multiply(A=@n19.RGB, B=VectorParameter() {parameterName=NM Intensity, defaultValue=<rgba 0.3 0.3 1 0>}.RGB)
  n20 = ScalarParameter() {defaultValue=2, parameterName=Fresnel Exponent A}
  n21 = ScalarParameter() {defaultValue=6, parameterName=Fresnel Exponent B}
  n22 = MaterialFunctionCall(Float2=TextureCoordinate()) {materialFunction=&BreakOutFloat2Components}
  NamedRerouteDeclaration(None=MaterialFunctionCall(Brightness=ScalarParameter() {defaultValue=1.2, parameterName=Fresnel Brightness}, Exponent=LinearInterpolate(A=@n20, B=@n21, Alpha=@n22.G)) {materialFunction=&MF_Fresnel_FX}.Fresnel Double Sided) {name=Double Sided Fresnel, nodeColor=<rgba 0 0.625 1 1>}
  n23 = Multiply(A=@n19.RGB, B=DynamicParameter() {paramNames [V Tile, Panner Speed, WPO Strength, Param4]}.WPO Strength)
  n24 = VectorParameter() {parameterName=WPO Flare, defaultValue=<rgba 15 15 0 0>}
  n25 = StaticSwitchParameter(True=@n24.RGB, False=Transform(None=@n24.RGB) {transformSourceType=TRANSFORMSOURCE_Local}) {parameterName=IceTrap}
  n26 = Multiply(A=Add(A=LinearInterpolate(A=@n23, Alpha=@n12) {constB=0}, B=Multiply(A=@n25, B=@n10.G)), B=VertexNormalWS())
  n27 = MaterialFunctionCall() {materialFunction=&ObjectPivotPoint}
  n28 = WorldPosition()
  StaticSwitchParameter(True=@n26, False=If(A=@n18.Distance, B=@n17, "A > B"=@n26, "A == B"=@n26, "A < B"=Subtract(A=@n27.Object Pivot Location, B=@n28.XYZ))) {parameterName=IceTrap}
 ]
}
